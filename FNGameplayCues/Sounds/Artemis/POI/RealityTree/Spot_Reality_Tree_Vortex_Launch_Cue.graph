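# Sound cue for the Reality Tree wind vortex launch whoosh.
# It picks one of four equally weighted whoosh waves at random and runs it through a modulator.
# It references the Reality Forest geyser launch attenuation asset and plays in the traps sound class with boosted volume.
# A concurrency override caps it at two simultaneous instances.
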SoundCue /FNGameplayCues/Sounds/Artemis/POI/RealityTree/Spot_Reality_Tree_Vortex_Launch_Cue
root=Modulator_1
nodes [
 {id=Modulator_1, children [Random_0]}
 {id=Random_0, children [WavePlayer_0,WavePlayer_1,WavePlayer_2,WavePlayer_3], properties {weights [1,1,1,1]}}
 {id=WavePlayer_0, properties {soundWaveAssetPtr=&/Reality_Tree_Wind_Vortex_Launch_Whoosh_01}}
 {id=WavePlayer_1, properties {soundWaveAssetPtr=&/Reality_Tree_Wind_Vortex_Launch_Whoosh_02}}
 {id=WavePlayer_2, properties {soundWaveAssetPtr=&/Reality_Tree_Wind_Vortex_Launch_Whoosh_03}}
 {id=WavePlayer_3, properties {soundWaveAssetPtr=&/Reality_Tree_Wind_Vortex_Launch_Whoosh_04}}
]
properties {
 attenuationSettings=/Game/Sounds/Artemis/Environment_Actors/RealityForest_Geyser/RealityForest_Daisy_Geyser_Launch_Att
 bOverrideConcurrency=true
 maxDistance=5500
 soundClassObject=&Fort_Traps
 volumeMultiplier=1.5
 attenuationOverrides {bEnableListenerFocus=true, falloffDistance=2000, attenuationShapeExtents {x=500}}
 concurrencyOverrides {maxCount=2}
}
